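# Translucent unlit effect material for the bow and arrow in the SeleneCobra player emote, usable on skeletal meshes.
# A view-angle Fresnel term blends color and opacity from set A at the edges to B facing the camera.
# Color feeds a shared emissive-adjustment function, while alpha is multiplied by a rerouted mask.
# The Arrow Tip Mask is an inverted sphere mask on interpolated pre-skinned position, keeping it fixed to the mesh through animation.
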Material /Game/Animation/Game/MainPlayer/Emotes/SeleneCobra/FX/Materials/M_Emote_SeleneCobra_BowAndArrow
graph {
 properties {
  bUsedWithMeshDeformer=true
  bUsedWithSkeletalMesh=true
  blendMode=BLEND_Translucent
  pixelDepthOffsetMode=PDOM_Legacy
  shadingModel=MSM_Unlit
  usedShadingModels=Unlit
 }
 nodes [
  n1 = VectorParameter() {parameterName=Color and Opacity A, defaultValue=<rgba 0 0.6 2 1>}
  n2 = VectorParameter() {parameterName=Color and Opacity B, defaultValue=<rgba 0.175 0 1.5 0.1>}
  n3 = VertexNormalWS()
  n4 = CameraVectorWS()
  n5 = LinearInterpolate(
   A=AppendVector(A=@n1.RGB, B=@n1.RGB)
   B=AppendVector(A=@n2.RGB, B=@n2.RGB)
   Alpha=Power(Base=DotProduct(A=@n3, B=@n4), Exp=ScalarParameter() {defaultValue=0.666, parameterName=Fresnel Exponent})
  )
  MaterialFunctionCall(Emissive=ComponentMask(None=@n5) {b=true, g=true, r=true}) {materialFunction=&MF_FX_EmissiveAdjustments}
  Multiply(A=ComponentMask(None=@n5) {a=true}, B=NamedRerouteUsage())
  n6 = VertexInterpolator(VS=PreSkinnedPosition())
  n7 = VectorParameter() {parameterName=MaskPositionAndRadius, defaultValue=<rgba 0 15.333 0 16>}
  n8 = ScalarParameter() {defaultValue=0.5, parameterName=Hardness}
  NamedRerouteDeclaration(None=OneMinus(None=SphereMask(A=@n6.PS, B=@n7.RGB, Radius=@n7.RGB, Hardness=@n8))) {name=Arrow Tip Mask, nodeColor=<rgba 0.8516 1 0 1>}
 ]
}
